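; Blueprint for the small Reality Seed consumable item that manages its held gameplay cue.
; When weapon visibility changes, it locally adds the looping GC_Held cue on itself if visible, or removes it if hidden.
; The construction script only calls its parent.
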(event EventOnWeaponVisibilityChanged
  (params (bVisible :type Boolean) (bSetForLocalControllerOnly :type Boolean))
  (bind _returnvalue (Ability|GameplayCue|MakeGameplayCueParameters 0.0))
  (bind _self self)
  (bind _gc_held (Variables|Default|GetGC_Held))
  (if bVisible
    (Projectile|AddGameplayCue%28Looping%29Local _self _gc_held _returnvalue)
    (else
      (Projectile|RemoveGameplayCue%28Looping%29Local _self _gc_held _returnvalue))))

(fn ConstructionScript
  (|CallParentFunction))
