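; Reality Seed projectile that spawns a sapling spawner on impact, based on the Athena BGA spawner projectile.
; SpawnBGA runs once, places the spawner at the impact point plus offset with a surface-derived yaw, and passes along persistent sapling data for transplants.
; It then sends a gameplay event to the player pawn and hides the projectile.
; Helpers check slope against the class-default threshold and test points against cached playable bounds.
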(fn IsPointinPlayableBounds
  (params (Point :type Vector))
  (Utilities|FlowControl|Sequence
    (:then_0
      (if (Math|Vector|VectorIsZero (Variables|Default|GetCachedPlayableBoundsBoxExtent))
        (bind (_playableboundsorigin _playableboundsextent) (Class|RealitySeedFunctionLibrary|GetPlayableBounds))
        (Variables|Default|SetCachedPlayableBoundsOrigin _playableboundsorigin)
        (Variables|Default|SetCachedPlayableBoundsBoxExtent _playableboundsextent)))
    (:then_1
      (return (Math|Geometry|IsPointinBox Point (Variables|Default|GetCachedPlayableBoundsOrigin) (Variables|Default|GetCachedPlayableBoundsBoxExtent))))))

(fn IsSlopeValid
  (params (Hit :type Hit%20Result%20Structure))
  (bind (_bblockinghit _binitialoverlap _time _distance _location _impactpoint _normal _impactnormal) (Collision|BreakHitResult Hit))
  (bind (_bgatospawn _soundactivate _spawnscale _spawnposoffsetalongnormal _spawnposzoffsetinwater _spawned _bspawnoncollision _mindotwithup) (Class|GetClassDefaults (Utilities|Casting|CastToB_Prj_Athena_BGASpawnerClass "/RealitySeedGameplay/items/RealitySeed/B_Prj_RealitySeed_SpawnSappling.B_Prj_RealitySeed_SpawnSappling_C")))
  (return (> (Math|Vector|DotProduct _impactnormal "0, 0,1.000000") _mindotwithup)))

(fn SpawnBGA_
  (params (Normal :type Vector :default "0, 0, 0") (HitActor :type Actor%20Object%20Reference) (HitComp :type Primitive%20Component%20Object%20Reference) (Pos :type Vector :default "0, 0, 0") (FromWater :type Boolean) (Hit :type Hit%20Result%20Structure))
  (bind (_bblockinghit _binitialoverlap _time _distance _location _impactpoint _normal _impactnormal) (Collision|BreakHitResult Hit))
  (bind _returnvalue (Transformation|GetActorRotation))
  (bind _returnvalue_2 (Math|Vector|GetForwardVector _returnvalue))
  (bind (_returnvalue_roll _returnvalue_pitch _returnvalue_yaw) (Math|Rotator|BreakRotator (Math|Rotator|MakeRotfromZX _impactnormal _returnvalue_2)))
  (bind _defaultspawnoffset (Variables|Default|GetDefaultSpawnOffset))
  (bind _returnvalue_3 (+ _impactpoint _defaultspawnoffset))
  (bind _playerpawn (Variables|Default|GetPlayerPawn))
  (bind (_persistent_sapling_data_owningplayercontroller _persistent_sapling_data_location _persistent_sapling_data_rotation _persistent_sapling_data_nurturinglevel _persistent_sapling_data_maturitylevel _persistent_sapling_data_plantingtime _persistent_sapling_data_lastweedingtime _persistent_sapling_data_fruitstates) (Utilities|Struct|BreakRealitySaplingPersistentData (Variables|GetPersistent%20Sapling%20Data_OwningPlayerController)))
  ;[0 1 0.2985] Determine if we should spawn a rift or a sapling based on their component tracking if the player has a sapling in play. Spawn the rift or sapling and send the object in an event to the player.
  (if (not (Variables|Logic|GetSpawnDoOnce))
    (Variables|Logic|SetSpawnDoOnce true)
    (bind _playerpawn_2 (Variables|Default|GetPlayerPawn)
      (:then
        (Utilities|IsValid _persistent_sapling_data_owningplayercontroller
          (:Is%20Valid
            (Variables|Default|SetPersistent%20Sapling%20Data (Utilities|Struct|MakeRealitySaplingPersistentData _persistent_sapling_data_owningplayercontroller _returnvalue_3 (Math|Rotator|MakeRotator 0.0 0.0 _returnvalue_yaw) _persistent_sapling_data_nurturinglevel _persistent_sapling_data_maturitylevel _persistent_sapling_data_plantingtime _persistent_sapling_data_lastweedingtime _persistent_sapling_data_fruitstates)))
          (:Is%20Not%20Valid
            _))
        (bind _returnvalue_4 (Game|SpawnActorfromClass :Class "/RealitySeedGameplay/items/RealitySaplings/B_RealitySaplings_Spawner.B_RealitySaplings_Spawner_C" :SpawnTransform (Math|Transform|MakeTransform _returnvalue_3 (Math|Rotator|MakeRotator 0.0 0.0 _returnvalue_yaw) (Transformation|GetActorScale3D)) :CollisionHandlingOverride "AlwaysSpawn" :TransformScaleMethod "MultiplyWithRoot" :Overrride%20SaplingPersistence false :OverrideIsLocalPlayerSapling false :Day%20Offset 0 :MovedSaplingPersistentData (Variables|Default|GetPersistent%20Sapling%20Data) :PlanterPlayerController (Projectile|GetInstigatorPlayerController) :SaplingNumTimesWateredThisMatch (Variables|Default|GetTransplantedSaplingNumTimesWateredThisMatch) :ClassToSpawn "/RealitySeedGameplay/items/RealitySaplings/B_SMA_Athena_RealitySapling_Base.B_SMA_Athena_RealitySapling_Base_C" :UseWeightedSpawnClassList false :Instigator (Variables|Default|GetPlayerPawn)))
        (Ability|SendGameplayEventtoActor _playerpawn_2 "(TagName=\"Event.Projectile.SpawnedBGA\")" (Utilities|Struct|MakeGameplayEventData "(TagName=\"Event.Projectile.SpawnedBGA\")" _playerpawn 0 _returnvalue_4))
        (Collision|SetActorEnableCollision false)
        (Rendering|SetActorHiddenInGame true)
        (Utilities|IsValid _returnvalue_4
          (:Is%20Valid
            _)
          (:Is%20Not%20Valid
            (Fortnite|FortShippingLog self (Utilities|String|Append "Spawn BGA failed. No sapling spawner spawned for player [ " (Fortnite|GetPlayerNameSafe _playerpawn) "]") true)))
        (return _returnvalue_4))
      (:else))))

(fn ConstructionScript
  (|CallParentFunction))
