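; Passive Reality Seed gameplay ability that detects a player's planted Reality Sapling tree and records it.
; On activation it waits for the SpawnedBGA projectile event, then binds to the sapling spawner's server actor-spawned delegate.
; On authority, RegisterTree registers the tree with the controller's RealitySapling component, flagging it new when moved-sapling persistent data is invalid.
; It sends a planted quest stat, plus a thrown-check stat when the tree lands far from the player.
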(event Ability|EventActivateAbility
  ;[0.3229 0.3229 0.3229] TEMP - Wait for the tree or rift to be spawned. Store them to determine if we should be spawning new rifts or not. Set the rift teleport target to the tree.
  (|CallParentFunction)
  (bind _outfortplayerpawn (CallFunction|GetAvatarasValidFortPlayerPawn)
    (:ValidFortPlayerPawn
      (Variables|Default|SetPlayerPawn _outfortplayerpawn)
      (bind (_asynctaskproxy _payload) (Ability|Tasks|WaitGameplayEvent "(TagName=\"Event.Projectile.SpawnedBGA\")")
        (:EventReceived
          (Variables|Default|GetPlayerPawn
            (:then
              (bind _asb_reality_saplings_spawner (Utilities|Casting|CastToB_RealitySaplings_Spawner (.OptionalObject (Utilities|Struct|BreakGameplayEventData _payload))))
              (Default|BindEventtoEventServerActorSpawned _asb_reality_saplings_spawner (delegate Custom|RegisterTree)))
            (:else)))))))

(event Ability|EventFailedToActivatePassiveAbility
  (params (ActorInfo :type Gameplay%20Ability%20Actor%20Info%20Structure))
  (|CallParentFunction))

(event Custom|RegisterTree
  (params (Actor :type Actor%20Object%20Reference))
  ;[0.04846 0.4062 0.2394] Store Tree - Called when the tree is spawned from a seed. Register it with the RealitySapling controller component for this player
  (Ability|HasAuthority
    (:True
      (bind _playerpawn (Variables|Default|GetPlayerPawn)
        (:then
          (Utilities|IsValid (Fortnite|GetFortPlayerControllerfromActor _playerpawn)
            (:Is%20Valid
              (Utilities|IsValid (Actor|GetComponentbyClass (Fortnite|GetFortPlayerControllerfromActor _playerpawn) "/Script/RealitySeedGameplayRuntime.FortControllerComponent_RealitySapling")
                (:Is%20Valid
                  (bind _asb_sma_athena_reality_sapling_parent (Utilities|Casting|CastToB_SMA_Athena_RealitySapling_Parent Actor))
                  (Class|FortControllerComponentRealitySapling|RegisterPlantedSapling (Actor|GetComponentbyClass (Fortnite|GetFortPlayerControllerfromActor _playerpawn) "/Script/RealitySeedGameplayRuntime.FortControllerComponent_RealitySapling") _asb_sma_athena_reality_sapling_parent (not (Utilities|IsValid (Utilities|Struct|BreakRealitySaplingPersistentData (Variables|/RealitySeedGameplay/items/RealitySaplings/B_SMA_Athena_RealitySapling_Parent.B_SMA_Athena_RealitySapling_Parent_C|GetMovedSaplingPersistentData _asb_sma_athena_reality_sapling_parent)) :!class /Script/Engine.KismetSystemLibrary)))
                  (CallFunction|SaplingPlanted (not (Utilities|IsValid (Utilities|Struct|BreakRealitySaplingPersistentData (Variables|/RealitySeedGameplay/items/RealitySaplings/B_SMA_Athena_RealitySapling_Parent.B_SMA_Athena_RealitySapling_Parent_C|GetMovedSaplingPersistentData _asb_sma_athena_reality_sapling_parent)) :!class /Script/Engine.KismetSystemLibrary)) _asb_sma_athena_reality_sapling_parent)
                  (Quest|SendComplexCustomStatEvent (Quests|GetQuestManager (Fortnite|GetFortPlayerControllerfromActor (Variables|Default|GetPlayerPawn)) "Athena") 0 "(GameplayTags=((TagName=\"Athena.Quests.RealitySapling.Planted\")))" "(GameplayTags=)" 1))))))
        (:else)))))

(event Custom|SaplingPlanted
  (params (isNewSapling :type Boolean) (Reality%20Tree :type B%20SMA%20Athena%20Reality%20Sapling%20Parent%20Object%20Reference))
  ;[1 1 1] Event for tracking when a sapling is planted - for quests
  (if (>= (Math|Vector|Distance(Vector) (Transformation|GetActorLocation Reality%20Tree) (Transformation|GetActorLocation (Variables|Default|GetPlayerPawn))) 3e+03)
    (Quest|SendComplexCustomStatEvent (Quests|GetQuestManager (Fortnite|GetFortPlayerControllerfromActor (Variables|Default|GetPlayerPawn)) "Athena") 0 "(GameplayTags=((TagName=\"Athena.Quests.RealitySeedThrownCheck\")))" "(GameplayTags=)" 1)))
