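; Gameplay ability for the RealitySeed item that, on activation, applies a tree-listener gameplay effect to the player, marked as temporary in its comment.
; It resolves the avatar as a valid Fortnite player pawn and applies GE_RealitySeed_ApplyTreeListener only when that effect is not already active, then ends.
; Per the comment, the effect grants an ability listening for tree/rift spawn, and the count check guards against players with a tree already out.
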(event Ability|EventActivateAbility
  ;[0.4479 0.4479 0.4479] TEMP - Apply GE that will apply GA to listen for the tree/rift spawn. Do not apply if the player has a tree out already.
  (|CallParentFunction)
  (bind _outfortplayerpawn (CallFunction|GetAvatarasValidFortPlayerPawn)
    (:ValidFortPlayerPawn
      (if (< (GameplayEffects|GetGameplayEffectCount (Ability|GetAbilitySystemComponent _outfortplayerpawn) "/RealitySeedGameplay/items/RealitySeed/GE_RealitySeed_ApplyTreeListener.GE_RealitySeed_ApplyTreeListener_C" 0 true) 1)
        (Ability|ApplyGameplayEffectToOwner "/RealitySeedGameplay/items/RealitySeed/GE_RealitySeed_ApplyTreeListener.GE_RealitySeed_ApplyTreeListener_C")
        (else
          _))
      (Ability|EndAbility))
    (:AvatarCastFailed
      _)))

(event Ability|EventFailedToActivatePassiveAbility
  (params (ActorInfo :type Gameplay%20Ability%20Actor%20Info%20Structure))
  (|CallParentFunction))
