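; Function library for Reality Seed items, providing shared queries and helpers for sapling placement, ownership and player bouncing.
; SurfaceSupportsSaplingDeployment returns an integer code from road-material and actor-class checks, box traces for obstacles and players, and a nearby-sapling overlap.
; Helpers find the player's sapling controller component, test playable bounds with optional caller-cached extents, and align rotations to surface normals.
; The bounce launches pawns within a cone above the actor, playing a local GameplayCue.
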(fn DoesPlayerHaveSapling
  (params (TargetActor :type Actor%20Object%20Reference))
  ;[0.1067 0.5288 1] Does the player have an active reality sapling?
  (bind _asfort_player_pawn (Utilities|Casting|CastToFortPlayerPawn TargetActor)
    (:then
      (Utilities|IsValid (CallFunction|GetRealitySaplingControllerComponent _asfort_player_pawn _asfort_player_pawn)
        (:Is%20Valid
          (return (Class|FortControllerComponentRealitySapling|HasActiveSapling (CallFunction|GetRealitySaplingControllerComponent _asfort_player_pawn _asfort_player_pawn))))
        (:Is%20Not%20Valid
          _)))
    (:CastFailed
      _))
  (return false))

(fn GetNearbySaplings
  (params (Location :type Vector))
  (bind (_outactors _returnvalue) (Collision|SphereOverlapActors 0 Location (ScalableFloat|GetValueatLevel (LocalVariables|GetSaplingOverlapRadius)) (Utilities|Array|MakeArray "ObjectTypeQuery1") "/RealitySeedGameplay/items/RealitySaplings/B_SMA_Athena_RealitySapling_Base.B_SMA_Athena_RealitySapling_Base_C"))
  (return _returnvalue))

(fn GetPlayableBounds
  (bind _worldcontext (Variables|Get__WorldContext))
  (bind _returnvalue (Fortnite|Athena|GameState|GetGameStateAthena _worldcontext))
  (bind _mapinfo (Variables|/Script/FortniteGame.FortGameStateAthena|GetMapInfo _returnvalue))
  (bind _playableboundsvolume (Variables|/Script/FortniteGame.FortAthenaMapInfo|GetPlayableBoundsVolume _mapinfo))
  (bind (_origin _boxextent) (Collision|GetActorBounds _playableboundsvolume false false))
  (Utilities|IsValid _returnvalue
    (:Is%20Valid
      (Utilities|IsValid _mapinfo
        (:Is%20Valid
          (Utilities|IsValid _playableboundsvolume
            (:Is%20Valid
              (return _origin _boxextent))
            (:Is%20Not%20Valid
              ;[1 0.2224 0.1793] Missing playable bounds
              (Development|PrintString 0 "Missing Playable Bounds!")
              (return "0, 0, 0" "0, 0, 0"))))
        (:Is%20Not%20Valid
          _)))
    (:Is%20Not%20Valid
      _))
  ;[1 0.09988 0.1364] No Gamestate, No Map Info
  (return "0, 0, 0" "0, 0, 0"))

(fn GetRealitySaplingControllerComponent
  (params (PlayerTarget :type Actor%20Object%20Reference))
  (bind _returnvalue (Fortnite|GetFortPlayerControllerfromActor PlayerTarget))
  (bind _returnvalue_2 (Actor|GetComponentbyClass _returnvalue "/Script/RealitySeedGameplayRuntime.FortControllerComponent_RealitySapling"))
  ;[1 0.1637 0.7134] Get the RealitySaplingController component from the target
  (Utilities|IsValid PlayerTarget
    (:Is%20Valid
      (Utilities|IsValid _returnvalue
        (:Is%20Valid
          (Utilities|IsValid _returnvalue_2
            (:Is%20Valid
              (return _returnvalue_2))
            (:Is%20Not%20Valid
              _)))
        (:Is%20Not%20Valid
          _))
      (return 0))
    (:Is%20Not%20Valid
      _)))

(fn IsPointInPlayableBounds
  (params (Point :type Vector) (CachedPlayable%20BoundsOrigin :type Vector) (CachedPlayable%20BoundsExtent :type Vector))
  (bind _returnvalue (Fortnite|Athena|GameState|GetGameStateAthena 0))
  (bind _mapinfo (Variables|/Script/FortniteGame.FortGameStateAthena|GetMapInfo _returnvalue))
  (bind _playableboundsvolume (Variables|/Script/FortniteGame.FortAthenaMapInfo|GetPlayableBoundsVolume _mapinfo))
  (bind (_origin _boxextent) (Collision|GetActorBounds _playableboundsvolume false false))
  (Utilities|FlowControl|Sequence
    (:then_0
      (if (Math|Vector|VectorIsZero CachedPlayable%20BoundsExtent)
        ;[0.266 1 0.3507] If no cached value was passed in, fetch the bounds and origin
        (Utilities|IsValid _returnvalue
          (:Is%20Valid
            (Utilities|IsValid _mapinfo
              (:Is%20Valid
                (Utilities|IsValid _playableboundsvolume
                  (:Is%20Valid
                    (LocalVariables|SetOrigin _origin)
                    (LocalVariables|SetExtent _boxextent))
                  (:Is%20Not%20Valid
                    ;;[1 0.2224 0.1793] Missing playable bounds -> return false
                    (Development|PrintString 0 "Missing Playable Bounds!")
                    (return false "0, 0, 0" "0, 0, 0"))))
              (:Is%20Not%20Valid
                _)))
          (:Is%20Not%20Valid
            _))
        ;;[1 0.09988 0.1364] No Gamestate, No Map Info -> return false
        (return false "0, 0, 0" "0, 0, 0")
        (else
          ;[0.4437 1 0.5028] If there are cached passed in values, use those
          (LocalVariables|SetOrigin (Variables|GetCachedPlayable%20BoundsOrigin))
          (LocalVariables|SetExtent (Variables|GetCachedPlayable%20BoundsExtent)))))
    (:then_1
      ;[1 0.7525 0.4251] Return whether point is in the playable bounds box and the origin and extent so they can be cached if desired
      (return (Math|Geometry|IsPointinBox Point (LocalVariables|GetOrigin) (LocalVariables|GetExtent)) (LocalVariables|GetOrigin) (LocalVariables|GetExtent)))))

(fn IsValidSurface
  (params (HitActor :type Actor%20Object%20Reference))
  (bind (_asspline_mesh_actor _bsuccess) (Utilities|Casting|CastToSplineMeshActor HitActor))
  (bind (_asbuilding_prop _bsuccess_2) (Utilities|Casting|CastToBuildingProp HitActor))
  (bind (_asbuilding_stairs _bsuccess_3) (Utilities|Casting|CastToBuildingStairs HitActor))
  (bind (_asbuilding_wall _bsuccess_4) (Utilities|Casting|CastToBuildingWall HitActor))
  (bind (_asbuilding_gameplay_actor _bsuccess_5) (Utilities|Casting|CastToBuildingGameplayActor HitActor))
  (bind (_ascontrol_point_mesh_actor _bsuccess_6) (Utilities|Casting|CastToControlPointMeshActor HitActor))
  (bind (_aslandscape _bsuccess_7) (Utilities|Casting|CastToLandscape HitActor))
  (bind (_aslandscape_streaming_proxy _bsuccess_8) (Utilities|Casting|CastToLandscapeStreamingProxy HitActor))
  (bind (_asstatic_mesh_actor _bsuccess_9) (Utilities|Casting|CastToStaticMeshActor HitActor))
  (bind (_asbuilding_deco _bsuccess_10) (Utilities|Casting|CastToBuildingDeco HitActor))
  (return (and (Math|Boolean|ORBoolean _bsuccess_8 _bsuccess_7 _bsuccess_9 _bsuccess _bsuccess_6) (not (Math|Boolean|ORBoolean _bsuccess_10 _bsuccess_2 _bsuccess_5 _bsuccess_3 _bsuccess_4)))))

(fn RotatorFacingTargetAlongNormal
  (params (Start :type Vector%20%28by%20ref%29) (Target :type Vector%20%28by%20ref%29) (RotationAxis :type Vector) (Xdirection :type Float%20%28double-precision%29))
  (return (Math|Rotator|MakeRotfromZX RotationAxis (Math|Vector|GetForwardVector (Math|Rotator|CombineRotators (Math|Rotator|FindLookatRotation Start Target) (Math|Rotator|MakeRotfromZX RotationAxis (Math|Vector|MakeVector Xdirection)))))))

(fn SaplingBouncePlayerOnHit
  (params (HitPlayerPawn :type Fort%20Player%20Pawn%20Object%20Reference) (Tent :type Actor%20Object%20Reference) (Bounce%20GC :type Gameplay%20Tag%20Structure) (Hit%20Location :type Vector) (Hit%20Normal :type Vector) (Recently%20Launched :type Array%20of%20Fort%20Player%20Pawn%20Object%20References))
  (bind _returnvalue (Transformation|GetActorLocation HitPlayerPawn))
  (bind _returnvalue_2 (Transformation|GetActorLocation Tent))
  (bind _returnvalue_3 (Math|Rotator|FindLookatRotation _returnvalue_2 _returnvalue))
  (bind _returnvalue_4 (Math|Vector|GetForwardVector _returnvalue_3))
  (if (<= (Math|Trig|Acos(Degrees) (Math|Vector|DotProduct (Transformation|GetActorUpVector Tent) _returnvalue_4)) 5e+01)
    (bind _launchedpawn (Utilities|LaunchPlayerPawn HitPlayerPawn HitPlayerPawn (* (/ (+ _returnvalue_4 (* Hit%20Normal -1.0)) 2.0) (* (Math|Float|MapRangeClamped (Math|Vector|DotProduct (+ _returnvalue_4 (* Hit%20Normal -1.0)) (Transformation|GetVelocity HitPlayerPawn)) 0.0 1.0 1.0 0.5) (ScalableFloat|GetValueatLevel (LocalVariables|GetBounce%20Impulse)))) false false false true true))
    (GameplayCue|ExecuteGameplayCue%28Burst%29Local Tent Bounce%20GC (Ability|GameplayCue|MakeGameplayCueParameters :NormalizedMagnitude 2.2e+02 :RawMagnitude 0.0 :Location Hit%20Location :Normal (Math|Rotator|GetRotationXVector (Math|Random|RandomRotator true)) :Instigator Tent :GameplayEffectLevel 1 :AbilityLevel 1 :bReplicateLocationWhenUsingMinimalRepProxy false))
    (return _launchedpawn)))

(fn SurfaceSupportsSaplingDeployment?
  (params (Hit :type Hit%20Result%20Structure%20%28by%20ref%29) (MinDistanceFromObstacles :type Float%20%28double-precision%29) (ProjRotation :type Rotator) (CheckSpecificSurfaces :type Boolean))
  (bind (_bblockinghit _binitialoverlap _time _distance _location _impactpoint _normal _impactnormal) (Collision|BreakHitResult (Variables|GetHit)))
  (bind (_bblockinghit_2 _binitialoverlap_2 _time_2 _distance_2 _location_2 _impactpoint_2 _normal_2 _impactnormal_2 _physmat _hitactor) (Collision|BreakHitResult Hit))
  (bind (_bblockinghit_3 _binitialoverlap_3 _time_3 _distance_3 _location_3) (Collision|BreakHitResult (Variables|GetHit)))
  (bind (_bblockinghit_4 _binitialoverlap_4 _time_4 _distance_4 _location_4 _impactpoint_3 _normal_3 _impactnormal_3) (Collision|BreakHitResult (Variables|GetHit)))
  ;[0.03484 1 0.7066] Check if the surface we are going to land on is a valid surface to spawn a sapling
  (LocalVariables|Sethit%20actor _hitactor)
  (Utilities|FlowControl|Sequence
    (:then_0
      (if (Variables|GetCheckSpecificSurfaces)
        ;;[0.07785 0.3576 1] Check for roads and other surfaces
        (if (not (and (Utilities|Array|ContainsItem (LocalVariables|GetRoadPhysicsMaterials) _physmat) (not (ScalableFloat|GetValueasBool (LocalVariables|GetSapplingsAllowedOnRoads?)))))
          (if (CallFunction|IsValidSurface (LocalVariables|Gethit%20actor))
            _
            (else
              _))
          (else
            _))
        (return 1)
        (else
          (return 0))))
    (:then_1
      ;;[0.9941 0.1976 1] Check if there are nearby blocking objects in the area of exclusion
      (bind (_outhit _returnvalue) (Collision|BoxTraceForObjects 0 (+ _location (* _impactnormal (+ (ScalableFloat|GetValueatLevel (LocalVariables|GetObstacleDetectionZOffset)) (Variables|GetMinDistanceFromObstacles)))) (+ _location (* _impactnormal (+ (ScalableFloat|GetValueatLevel (LocalVariables|GetObstacleDetectionZOffset)) (Variables|GetMinDistanceFromObstacles)))) (Math|Vector|MakeVector (Variables|GetMinDistanceFromObstacles) (Variables|GetMinDistanceFromObstacles) (* (Variables|GetMinDistanceFromObstacles) (ScalableFloat|GetValueatLevel (LocalVariables|GetObstacleDetectionZScale)))) (Variables|GetProjRotation) (Utilities|Array|MakeArray "ObjectTypeQuery1" "ObjectTypeQuery5" "ObjectTypeQuery12")))
      (if _returnvalue
        (if (not (CallFunction|IsValidSurface (.HitActor (Collision|BreakHitResult _outhit))))
          (return 3)))
      (bind _saplingsnearby (CallFunction|GetNearbySaplings _location_3))
      (if _saplingsnearby
        (return 2)
        (else
          ;;[0 0.1683 1] Extra check for the player, we don't want to spawn the tree on top of you.
          (bind (_outhit_2 _returnvalue_2) (Collision|BoxTraceForObjects 0 (+ _location_4 (* _impactnormal_3 (+ (ScalableFloat|GetValueatLevel (LocalVariables|GetObstacleDetectionZOffset)) (ScalableFloat|GetValueatLevel (LocalVariables|GetMinDistanceFromPlayer))))) (+ _location_4 (* _impactnormal_3 (+ (ScalableFloat|GetValueatLevel (LocalVariables|GetObstacleDetectionZOffset)) (ScalableFloat|GetValueatLevel (LocalVariables|GetMinDistanceFromPlayer))))) (Math|Vector|MakeVector (ScalableFloat|GetValueatLevel (LocalVariables|GetMinDistanceFromPlayer)) (ScalableFloat|GetValueatLevel (LocalVariables|GetMinDistanceFromPlayer)) (* (ScalableFloat|GetValueatLevel (LocalVariables|GetMinDistanceFromPlayer)) (ScalableFloat|GetValueatLevel (LocalVariables|GetObstacleDetectionZScale)))) (Variables|GetProjRotation) (Utilities|Array|MakeArray "ObjectTypeQuery3")))
          (return (Math|Integer|SelectInt 3 0 _returnvalue_2)))))))
