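# Sound cue for the Reality Seed item's bounce, routed through the impacts sound class.
# Picks one of four bounce waves at equal weight, then applies randomized upward pitch and a fixed volume boost after a short fixed delay.
# Overrides attenuation with a limited falloff range, and concurrency to allow only a couple of simultaneous instances per owner.
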SoundCue /RealitySeedGameplay/items/RealitySeed/Sounds/RealitySeed_Bounce_Cue
root=Delay_0
nodes [
 {id=Delay_0, children [Modulator_1], properties {delayMax=0.1, delayMin=0.1}}
 {id=Modulator_1, children [Random_0], properties {pitchMax=1.3, pitchMin=1, volumeMax=2, volumeMin=2}}
 {id=Random_0, children [WavePlayer_0,WavePlayer_3,WavePlayer_4,WavePlayer_5], properties {weights [1,1,1,1]}}
 {id=WavePlayer_0, properties {soundWaveAssetPtr=&/Reality_Seed_Bounce_04}}
 {id=WavePlayer_3, properties {soundWaveAssetPtr=&/Reality_Seed_Bounce_01}}
 {id=WavePlayer_4, properties {soundWaveAssetPtr=&/Reality_Seed_Bounce_02}}
 {id=WavePlayer_5, properties {soundWaveAssetPtr=&/Reality_Seed_Bounce_03}}
]
properties {
 attenuationSettings=/Game/Sounds/Attenuation/Fort_Bullet_Impact_Athena_Att
 bOverrideAttenuation=true
 bOverrideConcurrency=true
 maxDistance=1600
 soundClassObject=&Fort_Impacts
 volumeMultiplier=1
 attenuationOverrides {falloffDistance=1500, attenuationShapeExtents {x=100}}
 concurrencyOverrides {bLimitToOwner=true, maxCount=2}
}
