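# Sound cue for collecting a Reality Seed pod, reusing ice puck toy landing sounds.
# A Random node picks one of three equally weighted landing waves and feeds it through a Modulator for variation.
# It plays at reduced volume in the gadget items sound class, spatialized with the Reality Pods prop attenuation.
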SoundCue /RealitySeedGameplay/items/RealitySeed/Sounds/RealitySeed_Collect_Pod_Cue
root=Modulator_0
nodes [
 {id=Modulator_0, children [Random_0]}
 {id=Random_0, children [WavePlayer_0,WavePlayer_1,WavePlayer_2], properties {weights [1,1,1]}}
 {id=WavePlayer_0, properties {soundWaveAssetPtr=/Game/Athena/Sounds/Toys/Toy_IcePuck_Land_01}}
 {id=WavePlayer_1, properties {soundWaveAssetPtr=/Game/Athena/Sounds/Toys/Toy_IcePuck_Land_03}}
 {id=WavePlayer_2, properties {soundWaveAssetPtr=/Game/Athena/Sounds/Toys/Toy_IcePuck_Land_02}}
]
properties {
 attenuationSettings=/Game/Sounds/Attenuation/Fort_Prop_RealityPods_Att
 maxDistance=1600
 soundClassObject=&Fort_Gadgets_Items
 volumeMultiplier=0.6
}
