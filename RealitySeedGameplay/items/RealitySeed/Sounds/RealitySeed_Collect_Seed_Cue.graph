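# Sound cue played for collecting a Reality Seed, layering a collect sound with a whoosh.
# A mixer blends two branches, each randomly picking one of three equally weighted waves: collect variants at reduced volume and land-whoosh variants boosted louder.
# Each branch's modulator applies a fixed gain rather than random variation.
# It plays through the gadget items sound class with reality pod prop attenuation and a reduced overall volume.
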SoundCue /RealitySeedGameplay/items/RealitySeed/Sounds/RealitySeed_Collect_Seed_Cue
root=Mixer_0
nodes [
 {id=Mixer_0, children [Modulator_2,Modulator_1], properties {inputVolume [1,1]}}
 {id=Modulator_1, children [Random_1], properties {volumeMax=1.5, volumeMin=1.5}}
 {id=Modulator_2, children [Random_2], properties {volumeMax=0.7, volumeMin=0.7}}
 {id=Random_1, children [WavePlayer_1,WavePlayer_2,WavePlayer_0], properties {weights [1,1,1]}}
 {id=Random_2, children [WavePlayer_11, WavePlayer_12, WavePlayer_13], properties {weights [1,1,1]}}
 {id=WavePlayer_0, properties {soundWaveAssetPtr=&/Reality_Seed_Land_Whoosh_03}}
 {id=WavePlayer_1, properties {soundWaveAssetPtr=&/Reality_Seed_Land_Whoosh_01}}
 {id=WavePlayer_11, properties {soundWaveAssetPtr=&/Reality_Seed_Collect_01}}
 {id=WavePlayer_12, properties {soundWaveAssetPtr=&/Reality_Seed_Collect_02}}
 {id=WavePlayer_13, properties {soundWaveAssetPtr=&/Reality_Seed_Collect_03}}
 {id=WavePlayer_2, properties {soundWaveAssetPtr=&/Reality_Seed_Land_Whoosh_02}}
]
properties {
 attenuationSettings=/Game/Sounds/Attenuation/Fort_Prop_RealityPods_Att
 maxDistance=1600
 soundClassObject=&Fort_Gadgets_Items
 volumeMultiplier=0.5
}
