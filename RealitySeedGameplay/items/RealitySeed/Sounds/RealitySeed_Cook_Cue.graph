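# Sound cue for the Reality Seed "cook" sound in RealitySeedGameplay, played as player foley at reduced volume.
# A mixer layers a short leather-armor clothing swish over a random, equally weighted pick of two seed cook waves.
# Each layer passes through its own modulator, and the cue overrides attenuation with its own short-range falloff settings.
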SoundCue /RealitySeedGameplay/items/RealitySeed/Sounds/RealitySeed_Cook_Cue
root=Mixer_0
nodes [
 {id=Mixer_0, children [Modulator_0,Modulator_2], properties {inputVolume [1,1]}}
 {id=Modulator_0, children [WavePlayer_0]}
 {id=Modulator_2, children [Random_1]}
 {id=Random_1, children [WavePlayer_4,WavePlayer_5], properties {weights [1,1]}}
 {
  id=WavePlayer_0
  properties {soundWaveAssetPtr=/Game/Sounds/Fort_Player/Clothing/LeatherArmor/Waves/Foley_Clothing_LeatherArmor_SwishShort_05}
 }
 {id=WavePlayer_4, properties {soundWaveAssetPtr=&/Seed_Cook_01}}
 {id=WavePlayer_5, properties {soundWaveAssetPtr=&/Seed_Cook_02}}
]
properties {
 attenuationSettings=&Foley_Attenuation_Basic_Att
 bOverrideAttenuation=true
 maxDistance=1300
 soundClassObject=&Fort_Player_Foley_Athena
 volumeMultiplier=0.5
 attenuationOverrides {falloffDistance=1200, attenuationShapeExtents {x=100}}
}
