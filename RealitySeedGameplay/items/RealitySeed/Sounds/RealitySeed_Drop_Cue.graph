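# Sound cue played when a Reality Seed item is dropped.
# A Random node picks one of five equally weighted drop wave variations and feeds a Modulator at the root.
# The cue uses the shared weapon-drop attenuation and the gadget-items sound class, with a lowered overall volume.
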SoundCue /RealitySeedGameplay/items/RealitySeed/Sounds/RealitySeed_Drop_Cue
root=Modulator_0
nodes [
 {id=Modulator_0, children [Random_0]}
 {id=Random_0, children [WavePlayer_3,WavePlayer_4,WavePlayer_5,WavePlayer_6,WavePlayer_7], properties {weights [1,1,1,1,1]}}
 {id=WavePlayer_3, properties {soundWaveAssetPtr=&/RealitySeed_Drop_01}}
 {id=WavePlayer_4, properties {soundWaveAssetPtr=&/RealitySeed_Drop_02}}
 {id=WavePlayer_5, properties {soundWaveAssetPtr=&/RealitySeed_Drop_03}}
 {id=WavePlayer_6, properties {soundWaveAssetPtr=&/RealitySeed_Drop_04}}
 {id=WavePlayer_7, properties {soundWaveAssetPtr=&/RealitySeed_Drop_05}}
]
properties {
 attenuationSettings=&Fort_Weapon_Drop_Att
 maxDistance=3000
 soundClassObject=&Fort_Gadgets_Items
 volumeMultiplier=0.6
}
