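# Looping sound cue for the Reality Seed item while it is in the air.
# A looping wave player feeds a modulator that pins pitch well above normal and halves volume, with no random variation.
# A Doppler source effect chain shifts the sound with motion.
# It plays in the gadget items sound class with attenuation based on the reality pods preset, with overrides for falloff and inner shape.
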SoundCue /RealitySeedGameplay/items/RealitySeed/Sounds/RealitySeed_InAir_Loop_Cue
root=Modulator_0
nodes [
 {id=Modulator_0, children [WavePlayer_3], properties {pitchMax=3, pitchMin=3, volumeMax=0.5, volumeMin=0.5}}
 {id=WavePlayer_3, properties {bLooping=true, soundWaveAssetPtr=&/Reality_Seed_InAir_Loop}}
]
properties {
 attenuationSettings=/Game/Sounds/Attenuation/Fort_Prop_RealityPods_Att
 bOverrideAttenuation=true
 maxDistance=1600
 soundClassObject=&Fort_Gadgets_Items
 sourceEffectChain=&/SeedFX_Doppler_Preset
 volumeMultiplier=1
 attenuationOverrides {falloffDistance=1500, attenuationShapeExtents {x=100}}
}
