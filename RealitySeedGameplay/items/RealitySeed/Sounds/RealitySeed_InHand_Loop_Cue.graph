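# Sound cue for the Reality Seed item in hand, pairing an equip sound with a sustained held loop.
# A mixer layers one of three random equip one-shots with a looping wave that an enveloper fades in over about half a second.
# Fixed modulators set both levels and raise the loop's pitch threefold.
# It uses the gadget items sound class and overrides attenuation for a short falloff.
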SoundCue /RealitySeedGameplay/items/RealitySeed/Sounds/RealitySeed_InHand_Loop_Cue
root=Mixer_0
nodes [
 {
  id=Enveloper_1
  children [Modulator_0]
  properties {pitchCurve {editorCurveData {keys [{time=0}]}}, volumeCurve {editorCurveData {keys [<at 0.0005 0>,<at 0.5 1>]}}}
 }
 {id=Mixer_0, children [Enveloper_1,Modulator_1], properties {inputVolume [1,1]}}
 {id=Modulator_0, children [WavePlayer_14], properties {pitchMax=3, pitchMin=3, volumeMax=0.2, volumeMin=0.2}}
 {id=Modulator_1, children [Random_0], properties {pitchMax=1, pitchMin=1, volumeMax=0.3, volumeMin=0.3}}
 {id=Random_0, children [WavePlayer_11, WavePlayer_12, WavePlayer_13], properties {weights [1,1,1]}}
 {id=WavePlayer_11, properties {soundWaveAssetPtr=&/Reality_Seed_Equip_01}}
 {id=WavePlayer_12, properties {soundWaveAssetPtr=&/Reality_Seed_Equip_02}}
 {id=WavePlayer_13, properties {soundWaveAssetPtr=&/Reality_Seed_Equip_03}}
 {id=WavePlayer_14, properties {bLooping=true, soundWaveAssetPtr=&/Reality_Seed_InHand_loop}}
]
properties {
 attenuationSettings=/Game/Sounds/Attenuation/Fort_Prop_RealityPods_Att
 bOverrideAttenuation=true
 maxDistance=200
 soundClassObject=&Fort_Gadgets_Items
 volumeMultiplier=0.7
 attenuationOverrides {falloffDistance=100, attenuationShapeExtents {x=100}}
}
