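# Sound cue for the Reality Seed item's landing sound, played in the gadgets and items sound class.
# A random node picks one of four equally weighted bounce waves and feeds it through a modulator at the root, which is left at its usual settings.
# Playback runs at reduced volume and uses the shared Reality Pods prop attenuation asset to limit its range.
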SoundCue /RealitySeedGameplay/items/RealitySeed/Sounds/RealitySeed_Land_Cue
root=Modulator_0
nodes [
 {id=Modulator_0, children [Random_0]}
 {id=Random_0, children [WavePlayer_3,WavePlayer_4,WavePlayer_5,WavePlayer_7], properties {weights [1,1,1,1]}}
 {id=WavePlayer_3, properties {soundWaveAssetPtr=&/Reality_Seed_Bounce_04}}
 {id=WavePlayer_4, properties {soundWaveAssetPtr=&/Reality_Seed_Bounce_01}}
 {id=WavePlayer_5, properties {soundWaveAssetPtr=&/Reality_Seed_Bounce_02}}
 {id=WavePlayer_7, properties {soundWaveAssetPtr=&/Reality_Seed_Bounce_03}}
]
properties {
 attenuationSettings=/Game/Sounds/Attenuation/Fort_Prop_RealityPods_Att
 maxDistance=1600
 soundClassObject=&Fort_Gadgets_Items
 volumeMultiplier=0.5
}
