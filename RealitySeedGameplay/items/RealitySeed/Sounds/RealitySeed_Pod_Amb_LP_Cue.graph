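# Looping ambient sound cue for the Reality Seed pod gadget item, with overridden attenuation and a small concurrency cap.
# A quality-level switch sends its first two slots to a random pick of three looping waves and its last two to the third loop alone, at higher volume.
# Continuous modulators name FallSpeed for volume but output a constant level, while their pitch range has no parameter name set.
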SoundCue /RealitySeedGameplay/items/RealitySeed/Sounds/RealitySeed_Pod_Amb_LP_Cue
root=QualityLevel_0
nodes [
 {
  id=ModulatorContinuous_2
  children [Random_0]
  properties {pitchModulationParams {maxOutput=1.25, minOutput=1}, volumeModulationParams {minOutput=1, parameterName=FallSpeed}}
 }
 {
  id=ModulatorContinuous_5
  children [WavePlayer_1]
  properties {pitchModulationParams {maxOutput=1.25, minOutput=1}, volumeModulationParams {maxOutput=1.4, minOutput=1.4, parameterName=FallSpeed}}
 }
 {id=QualityLevel_0, children [ModulatorContinuous_2, ModulatorContinuous_2, ModulatorContinuous_5, ModulatorContinuous_5]}
 {id=Random_0, children [WavePlayer_2,WavePlayer_0,WavePlayer_1], properties {weights [1,1,1]}}
 {id=WavePlayer_0, properties {bLooping=true, soundWaveAssetPtr=&/Reality_Seed_Pod_3D_Loop_02}}
 {id=WavePlayer_1, properties {bLooping=true, soundWaveAssetPtr=&/Reality_Seed_Pod_3D_Loop_03}}
 {id=WavePlayer_2, properties {bLooping=true, soundWaveAssetPtr=&/Reality_SeedPod_3D_Loop_1}}
]
properties {
 attenuationSettings=/Game/Sounds/Attenuation/Fort_Prop_RealityPods_Att
 bOverrideAttenuation=true
 bOverrideConcurrency=true
 maxDistance=1300
 soundClassObject=&Fort_Gadgets_Items
 volumeMultiplier=1
 attenuationOverrides {falloffDistance=900}
 concurrencyOverrides {maxCount=2}
}
