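# Sound cue for a Reality Seed pod's death, layering a burst sound with a juicy splatter sound.
# A mixer blends one of three random burst variants with one of four random juice variants.
# The juice layer starts after a short fixed delay with boosted volume, and each layer passes through its own modulator.
# The whole cue plays at reduced volume with shared toy-hit attenuation.
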SoundCue /RealitySeedGameplay/items/RealitySeed/Sounds/RealitySeed_Pod_Die_Cue
root=Mixer_0
nodes [
 {id=Delay_0, children [Modulator_1], properties {delayMax=0.1, delayMin=0.1}}
 {id=Mixer_0, children [Modulator_0,Delay_0], properties {inputVolume [1,1]}}
 {id=Modulator_0, children [Random_0]}
 {id=Modulator_1, children [Random_1], properties {volumeMax=2, volumeMin=2}}
 {id=Random_0, children [WavePlayer_1,WavePlayer_2,WavePlayer_3], properties {weights [1,1,1]}}
 {id=Random_1, children [WavePlayer_0,WavePlayer_4,WavePlayer_5,WavePlayer_6], properties {weights [1,1,1,1]}}
 {id=WavePlayer_0, properties {soundWaveAssetPtr=&/Reality_SeedPod_Burst_Juice_01}}
 {id=WavePlayer_1, properties {soundWaveAssetPtr=&/Reality_SeedPod_Burst_01}}
 {id=WavePlayer_2, properties {soundWaveAssetPtr=&/Reality_SeedPod_Burst_02}}
 {id=WavePlayer_3, properties {soundWaveAssetPtr=&/Reality_SeedPod_Burst_03}}
 {id=WavePlayer_4, properties {soundWaveAssetPtr=&/Reality_SeedPod_Burst_Juice_02}}
 {id=WavePlayer_5, properties {soundWaveAssetPtr=&/Reality_SeedPod_Burst_Juice_03}}
 {id=WavePlayer_6, properties {soundWaveAssetPtr=&/Reality_SeedPod_Burst_Juice_04}}
]
properties {attenuationSettings=/Game/Sounds/Attenuation/Fort_ToyHit_Att, maxDistance=2800, volumeMultiplier=0.5}
