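# Sound cue for bullet impacts on the Reality Seed pod.
# It picks one of ten equally weighted impact waves at random, passes it through a modulator for variation, then plays it after a short fixed delay.
# It uses Fort bullet impact attenuation and the impacts sound class, with boosted volume and overridden concurrency that caps simultaneous plays per owner.
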SoundCue /RealitySeedGameplay/items/RealitySeed/Sounds/RealitySeed_Pod_Impact_Bullet_Cue
root=Delay_0
nodes [
 {id=Delay_0, children [Modulator_0], properties {delayMax=0.1, delayMin=0.1}}
 {id=Modulator_0, children [Random_1], properties {pitchMax=1}}
 {
  id=Random_1
  children [
   WavePlayer_1
   WavePlayer_2
   WavePlayer_10
   WavePlayer_11
   WavePlayer_12
   WavePlayer_13
   WavePlayer_14
   WavePlayer_15
   WavePlayer_16
   WavePlayer_17
  ]
  properties {weights [1,1,1,1,1,1,1,1,1,1]}
 }
 {id=WavePlayer_1, properties {soundWaveAssetPtr=&/Reality_SeedPod_Bullet_Impact_01}}
 {id=WavePlayer_10, properties {soundWaveAssetPtr=&/Reality_SeedPod_Bullet_Impact_03}}
 {id=WavePlayer_11, properties {soundWaveAssetPtr=&/Reality_SeedPod_Bullet_Impact_04}}
 {id=WavePlayer_12, properties {soundWaveAssetPtr=&/Reality_SeedPod_Bullet_Impact_05}}
 {id=WavePlayer_13, properties {soundWaveAssetPtr=&/Reality_SeedPod_Bullet_Impact_06}}
 {id=WavePlayer_14, properties {soundWaveAssetPtr=&/Reality_SeedPod_Bullet_Impact_07}}
 {id=WavePlayer_15, properties {soundWaveAssetPtr=&/Reality_SeedPod_Bullet_Impact_08}}
 {id=WavePlayer_16, properties {soundWaveAssetPtr=&/Reality_SeedPod_Bullet_Impact_09}}
 {id=WavePlayer_17, properties {soundWaveAssetPtr=&/Reality_SeedPod_Bullet_Impact_10}}
 {id=WavePlayer_2, properties {soundWaveAssetPtr=&/Reality_SeedPod_Bullet_Impact_02}}
]
properties {
 attenuationSettings=/Game/Sounds/Attenuation/Fort_Bullet_Impact_Athena_Att
 bOverrideConcurrency=true
 maxDistance=4500
 soundClassObject=&Fort_Impacts
 volumeMultiplier=3
 concurrencyOverrides {bLimitToOwner=true, maxCount=2}
}
