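# Sound cue for the fire impact of a Reality Seed pod, reusing Zapper weapon ricochet fire-impact waves.
# A short fixed delay feeds a modulator that randomizes pitch upward, then a random node picks one of two waves with equal weight.
# It uses bullet-impact attenuation and the impacts sound class, with a concurrency override that caps simultaneous plays per owner.
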SoundCue /RealitySeedGameplay/items/RealitySeed/Sounds/RealitySeed_Pod_Impact_Fire_Cue
root=Delay_0
nodes [
 {id=Delay_0, children [Modulator_1], properties {delayMax=0.1, delayMin=0.1}}
 {id=Modulator_1, children [Random_0], properties {pitchMax=1.3, pitchMin=1}}
 {id=Random_0, children [WavePlayer_0,WavePlayer_3], properties {weights [1,1]}}
 {id=WavePlayer_0, properties {soundWaveAssetPtr=/Game/Sounds/Fort_Monster_Zapper/Zapper_Weapon_Impact_Rico_Fire_02}}
 {id=WavePlayer_3, properties {soundWaveAssetPtr=/Game/Sounds/Fort_Monster_Zapper/Zapper_Weapon_Impact_Rico_Fire_01}}
]
properties {
 attenuationSettings=/Game/Sounds/Attenuation/Fort_Bullet_Impact_Athena_Att
 bOverrideConcurrency=true
 maxDistance=4500
 soundClassObject=&Fort_Impacts
 concurrencyOverrides {bLimitToOwner=true, maxCount=2}
}
