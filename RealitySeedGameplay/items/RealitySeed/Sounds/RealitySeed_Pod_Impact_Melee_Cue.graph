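# Sound cue for melee impacts on the Reality Seed pod, named as the pod's melee hit sound.
# A mixer layers three randomized banks (main impact, bounce and juice), each picked with equal weights and passed through its own volume-boost and pitch-variation modulator.
# The main impact layer is boosted most with only slight pitch variation, while bounce and juice can drop slightly lower in pitch.
# It uses bullet-impact attenuation, the impacts sound class, and a per-owner concurrency cap.
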SoundCue /RealitySeedGameplay/items/RealitySeed/Sounds/RealitySeed_Pod_Impact_Melee_Cue
root=Mixer_0
nodes [
 {id=Mixer_0, children [Modulator_1,Modulator_0,Modulator_2], properties {inputVolume [1,1,1]}}
 {id=Modulator_0, children [Random_1], properties {pitchMax=1, pitchMin=0.9, volumeMax=2.5, volumeMin=2.5}}
 {id=Modulator_1, children [Random_0], properties {pitchMax=1.02, pitchMin=0.98, volumeMax=3, volumeMin=3}}
 {id=Modulator_2, children [Random_2], properties {pitchMax=1, pitchMin=0.9, volumeMax=2.5, volumeMin=2.5}}
 {
  id=Random_0
  children [WavePlayer_0, WavePlayer_4, WavePlayer_5, WavePlayer_6, WavePlayer_7, WavePlayer_8, WavePlayer_1, WavePlayer_2, WavePlayer_3, WavePlayer_15]
  properties {weights [1,1,1,1,1,1,1,1,1,1]}
 }
 {
  id=Random_1
  children [WavePlayer_9, WavePlayer_10, WavePlayer_11, WavePlayer_12, WavePlayer_13, WavePlayer_14]
  properties {weights [1,1,1,1,1,1]}
 }
 {
  id=Random_2
  children [WavePlayer_16, WavePlayer_17, WavePlayer_18, WavePlayer_19, WavePlayer_20, WavePlayer_21]
  properties {weights [1,1,1,1,1,1]}
 }
 {id=WavePlayer_0, properties {soundWaveAssetPtr=&/Reality_SeedPod_Melee_Impact_01}}
 {id=WavePlayer_1, properties {soundWaveAssetPtr=&/Reality_SeedPod_Melee_Impact_07}}
 {id=WavePlayer_10, properties {soundWaveAssetPtr=&/Reality_SeedPod_Melee_Impact_Bounce_02}}
 {id=WavePlayer_11, properties {soundWaveAssetPtr=&/Reality_SeedPod_Melee_Impact_Bounce_03}}
 {id=WavePlayer_12, properties {soundWaveAssetPtr=&/Reality_SeedPod_Melee_Impact_Bounce_04}}
 {id=WavePlayer_13, properties {soundWaveAssetPtr=&/Reality_SeedPod_Melee_Impact_Bounce_05}}
 {id=WavePlayer_14, properties {soundWaveAssetPtr=&/Reality_SeedPod_Melee_Impact_Bounce_06}}
 {id=WavePlayer_15, properties {soundWaveAssetPtr=&/Reality_SeedPod_Melee_Impact_10}}
 {id=WavePlayer_16, properties {soundWaveAssetPtr=&/Reality_SeedPod_Melee_Impact_Juice_01}}
 {id=WavePlayer_17, properties {soundWaveAssetPtr=&/Reality_SeedPod_Melee_Impact_Juice_02}}
 {id=WavePlayer_18, properties {soundWaveAssetPtr=&/Reality_SeedPod_Melee_Impact_Juice_03}}
 {id=WavePlayer_19, properties {soundWaveAssetPtr=&/Reality_SeedPod_Melee_Impact_Juice_04}}
 {id=WavePlayer_2, properties {soundWaveAssetPtr=&/Reality_SeedPod_Melee_Impact_08}}
 {id=WavePlayer_20, properties {soundWaveAssetPtr=&/Reality_SeedPod_Melee_Impact_Juice_05}}
 {id=WavePlayer_21, properties {soundWaveAssetPtr=&/Reality_SeedPod_Melee_Impact_Juice_06}}
 {id=WavePlayer_3, properties {soundWaveAssetPtr=&/Reality_SeedPod_Melee_Impact_09}}
 {id=WavePlayer_4, properties {soundWaveAssetPtr=&/Reality_SeedPod_Melee_Impact_02}}
 {id=WavePlayer_5, properties {soundWaveAssetPtr=&/Reality_SeedPod_Melee_Impact_03}}
 {id=WavePlayer_6, properties {soundWaveAssetPtr=&/Reality_SeedPod_Melee_Impact_04}}
 {id=WavePlayer_7, properties {soundWaveAssetPtr=&/Reality_SeedPod_Melee_Impact_05}}
 {id=WavePlayer_8, properties {soundWaveAssetPtr=&/Reality_SeedPod_Melee_Impact_06}}
 {id=WavePlayer_9, properties {soundWaveAssetPtr=&/Reality_SeedPod_Melee_Impact_Bounce_01}}
]
properties {
 attenuationSettings=/Game/Sounds/Attenuation/Fort_Bullet_Impact_Athena_Att
 bOverrideConcurrency=true
 maxDistance=4500
 soundClassObject=&Fort_Impacts
 volumeMultiplier=0.6
 concurrencyOverrides {bLimitToOwner=true, maxCount=2}
}
