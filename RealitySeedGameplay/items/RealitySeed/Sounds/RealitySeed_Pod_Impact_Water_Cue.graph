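# Sound cue for the Reality Seed pod hitting water, layering a seed bounce with a water splash.
# A Mixer blends two modulated Random branches: one of four bounce waves with a strong volume boost, and one of six shared heavy small-object splashes.
# It uses bullet-impact attenuation and the impacts sound class, with an owner-limited concurrency override that stops the oldest instance.
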SoundCue /RealitySeedGameplay/items/RealitySeed/Sounds/RealitySeed_Pod_Impact_Water_Cue
root=Mixer_0
nodes [
 {id=Mixer_0, children [Modulator_1,Modulator_0], properties {inputVolume [1,1]}}
 {id=Modulator_0, children [Random_2], properties {volumeMax=1, volumeMin=1}}
 {id=Modulator_1, children [Random_0], properties {pitchMin=1, volumeMax=5, volumeMin=5}}
 {id=Random_0, children [WavePlayer_0, WavePlayer_3, WavePlayer_4, WavePlayer_11], properties {weights [1,1,1,1]}}
 {
  id=Random_2
  children [WavePlayer_5, WavePlayer_6, WavePlayer_7, WavePlayer_8, WavePlayer_9, WavePlayer_10]
  properties {weights [1,1,1,1,1,1]}
 }
 {id=WavePlayer_0, properties {soundWaveAssetPtr=&/Reality_Seed_Bounce_01}}
 {
  id=WavePlayer_10
  properties {soundWaveAssetPtr=/FNGameplayCues/Sounds/Physics/Objects/WaterSplash/SmallObject/Phys_Small_Object_Splash_Heavy_05}
 }
 {id=WavePlayer_11, properties {soundWaveAssetPtr=&/Reality_Seed_Bounce_04}}
 {id=WavePlayer_3, properties {soundWaveAssetPtr=&/Reality_Seed_Bounce_02}}
 {id=WavePlayer_4, properties {soundWaveAssetPtr=&/Reality_Seed_Bounce_03}}
 {
  id=WavePlayer_5
  properties {soundWaveAssetPtr=/FNGameplayCues/Sounds/Physics/Objects/WaterSplash/SmallObject/Phys_Small_Object_Splash_Heavy_06}
 }
 {
  id=WavePlayer_6
  properties {soundWaveAssetPtr=/FNGameplayCues/Sounds/Physics/Objects/WaterSplash/SmallObject/Phys_Small_Object_Splash_Heavy_01}
 }
 {
  id=WavePlayer_7
  properties {soundWaveAssetPtr=/FNGameplayCues/Sounds/Physics/Objects/WaterSplash/SmallObject/Phys_Small_Object_Splash_Heavy_02}
 }
 {
  id=WavePlayer_8
  properties {soundWaveAssetPtr=/FNGameplayCues/Sounds/Physics/Objects/WaterSplash/SmallObject/Phys_Small_Object_Splash_Heavy_03}
 }
 {
  id=WavePlayer_9
  properties {soundWaveAssetPtr=/FNGameplayCues/Sounds/Physics/Objects/WaterSplash/SmallObject/Phys_Small_Object_Splash_Heavy_04}
 }
]
properties {
 attenuationSettings=/Game/Sounds/Attenuation/Fort_Bullet_Impact_Athena_Att
 bOverrideConcurrency=true
 maxDistance=4500
 soundClassObject=&Fort_Impacts
 concurrencyOverrides {bLimitToOwner=true, maxCount=4, resolutionRule=StopOldest}
}
