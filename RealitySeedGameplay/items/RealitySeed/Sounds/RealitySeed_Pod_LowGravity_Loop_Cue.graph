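# Sound cue for the Reality Seed pod's looping player low-gravity sound, classed as a gadget item sound.
# A Random node picks one of four equally weighted looping waves, and a Modulator lowers its pitch by a fixed amount on top of the cue's reduced pitch multiplier.
# The root parameter crossfade has a single input whose volume is governed by the FallSpeed parameter.
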SoundCue /RealitySeedGameplay/items/RealitySeed/Sounds/RealitySeed_Pod_LowGravity_Loop_Cue
root=ParamCrossFade_0
nodes [
 {id=Modulator_0, children [Random_0], properties {pitchMax=0.9, pitchMin=0.9, volumeMax=1, volumeMin=1}}
 {
  id=ParamCrossFade_0
  children [Modulator_0]
  properties {paramName=FallSpeed, crossFadeInput [{fadeInDistanceEnd=0, fadeInDistanceStart=0, fadeOutDistanceEnd=0, fadeOutDistanceStart=3, volume=1}]}
 }
 {id=Random_0, children [WavePlayer_2,WavePlayer_3,WavePlayer_4,WavePlayer_5], properties {weights [1,1,1,1]}}
 {
  id=WavePlayer_2
  properties {bLooping=true, soundWaveAssetPtr=&/Reality_SeedPod_PlayerLowGravity_Loop_01}
 }
 {
  id=WavePlayer_3
  properties {bLooping=true, soundWaveAssetPtr=&/Reality_SeedPod_PlayerLowGravity_Loop_02}
 }
 {
  id=WavePlayer_4
  properties {bLooping=true, soundWaveAssetPtr=&/Reality_SeedPod_PlayerLowGravity_Loop_03}
 }
 {
  id=WavePlayer_5
  properties {bLooping=true, soundWaveAssetPtr=&/Reality_SeedPod_PlayerLowGravity_Loop_04}
 }
]
properties {
 attenuationSettings=/Game/Sounds/Attenuation/Fort_Prop_RealityPods_Att
 bOverrideAttenuation=true
 maxDistance=4000
 pitchMultiplier=0.7
 soundClassObject=&Fort_Gadgets_Items
 volumeMultiplier=1
}
