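# Sound cue that plays when the Reality Seed pod's low gravity effect stops.
# A random node picks one of three low gravity end waves with equal weight, then a modulator scales it to a fixed, low volume.
# The sound is spatialized with the shared Reality Pods prop attenuation and routed through the gadget items sound class.
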SoundCue /RealitySeedGameplay/items/RealitySeed/Sounds/RealitySeed_Pod_LowGravity_Stop_Cue
root=Modulator_0
nodes [
 {id=Modulator_0, children [Random_1], properties {volumeMax=0.2, volumeMin=0.2}}
 {id=Random_1, children [WavePlayer_5,WavePlayer_6,WavePlayer_7], properties {weights [1,1,1]}}
 {id=WavePlayer_5, properties {soundWaveAssetPtr=&/Reality_SeedPod_LowGravityEnd_01}}
 {id=WavePlayer_6, properties {soundWaveAssetPtr=&/Reality_SeedPod_LowGravityEnd_02}}
 {id=WavePlayer_7, properties {soundWaveAssetPtr=&/Reality_SeedPod_LowGravityEnd_03}}
]
properties {
 attenuationSettings=/Game/Sounds/Attenuation/Fort_Prop_RealityPods_Att
 maxDistance=1600
 soundClassObject=&Fort_Gadgets_Items
 volumeMultiplier=1
}
