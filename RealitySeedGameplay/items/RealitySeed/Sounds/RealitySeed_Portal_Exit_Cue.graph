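# Sound cue for exiting a Reality Seed portal, classed as a gadget item sound.
# A Random node picks with equal weight between two Gravity Sphere exit waves borrowed from gameplay cue sounds.
# A root Modulator passes that pick through, and playback uses the Reality Pods prop attenuation.
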SoundCue /RealitySeedGameplay/items/RealitySeed/Sounds/RealitySeed_Portal_Exit_Cue
root=Modulator_0
nodes [
 {id=Modulator_0, children [Random_1]}
 {id=Random_1, children [WavePlayer_7,WavePlayer_9], properties {weights [1,1]}}
 {id=WavePlayer_7, properties {soundWaveAssetPtr=/FNGameplayCues/Sounds/Fort_Gadgets/GravitySphere/fort_gravity_sphere_exitnew_01}}
 {id=WavePlayer_9, properties {soundWaveAssetPtr=/FNGameplayCues/Sounds/Fort_Gadgets/GravitySphere/fort_gravity_sphere_exitnew_03}}
]
properties {
 attenuationSettings=/Game/Sounds/Attenuation/Fort_Prop_RealityPods_Att
 maxDistance=1600
 soundClassObject=&Fort_Gadgets_Items
 volumeMultiplier=1
}
