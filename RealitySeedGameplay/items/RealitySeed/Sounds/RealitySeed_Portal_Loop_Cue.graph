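# Sound cue for the looping portal sound of the Reality Seed item.
# A wave player loops the Reality_SeedPod_3D_Loop_1 wave and feeds a Modulator node at the root.
# It uses the Reality Pods prop attenuation asset for spatial falloff and routes through the Gadgets Items sound class.
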SoundCue /RealitySeedGameplay/items/RealitySeed/Sounds/RealitySeed_Portal_Loop_Cue
root=Modulator_1
nodes [
 {id=Modulator_1, children [WavePlayer_2]}
 {id=WavePlayer_2, properties {bLooping=true, soundWaveAssetPtr=&/Reality_SeedPod_3D_Loop_1}}
]
properties {
 attenuationSettings=/Game/Sounds/Attenuation/Fort_Prop_RealityPods_Att
 maxDistance=1600
 soundClassObject=&Fort_Gadgets_Items
 volumeMultiplier=1
}
