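# Sound cue for the Reality Seed being destroyed, built from the seed's landing sounds.
# A mixer layers a random pick of four land sounds with a random pick of three land whooshes, each through its own modulator, with the whoosh layer boosted in volume.
# It plays in the impacts sound class, with overridden attenuation for distance falloff and a per-owner concurrency limit of two.
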SoundCue /RealitySeedGameplay/items/RealitySeed/Sounds/RealitySeed_Seed_Destroyed_Cue
root=Mixer_0
nodes [
 {id=Mixer_0, children [Modulator_0,Modulator_1], properties {inputVolume [1,1]}}
 {id=Modulator_0, children [Random_1]}
 {id=Modulator_1, children [Random_0], properties {volumeMax=2, volumeMin=2}}
 {id=Random_0, children [WavePlayer_1,WavePlayer_2,WavePlayer_3], properties {weights [1,1,1]}}
 {id=Random_1, children [WavePlayer_0,WavePlayer_4,WavePlayer_5,WavePlayer_6], properties {weights [1,1,1,1]}}
 {id=WavePlayer_0, properties {soundWaveAssetPtr=&/Reality_Seed_Land_01}}
 {id=WavePlayer_1, properties {soundWaveAssetPtr=&/Reality_Seed_Land_Whoosh_01}}
 {id=WavePlayer_2, properties {soundWaveAssetPtr=&/Reality_Seed_Land_Whoosh_02}}
 {id=WavePlayer_3, properties {soundWaveAssetPtr=&/Reality_Seed_Land_Whoosh_03}}
 {id=WavePlayer_4, properties {soundWaveAssetPtr=&/Reality_Seed_Land_02}}
 {id=WavePlayer_5, properties {soundWaveAssetPtr=&/Reality_Seed_Land_03}}
 {id=WavePlayer_6, properties {soundWaveAssetPtr=&/Reality_Seed_Land_04}}
]
properties {
 attenuationSettings=/Game/Sounds/Attenuation/Fort_Bullet_Impact_Athena_Att
 bOverrideAttenuation=true
 bOverrideConcurrency=true
 maxDistance=2600
 soundClassObject=&Fort_Impacts
 volumeMultiplier=1
 attenuationOverrides {falloffDistance=2500, attenuationShapeExtents {x=100}}
 concurrencyOverrides {bLimitToOwner=true, maxCount=2}
}
