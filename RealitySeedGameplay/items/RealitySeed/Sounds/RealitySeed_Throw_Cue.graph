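# Sound cue for throwing the Reality Seed item, played as player foley with overridden attenuation.
# A mixer layers three randomized banks at once: land whooshes, launch sounds at reduced level, and seed throw sounds.
# Each bank passes through its own modulator, which pitches up the whoosh and launch layers and slightly varies the throw layer's volume.
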SoundCue /RealitySeedGameplay/items/RealitySeed/Sounds/RealitySeed_Throw_Cue
root=Mixer_1
nodes [
 {id=Mixer_1, children [Modulator_1,Modulator_2,Modulator_3], properties {inputVolume [1,0.5,1]}}
 {id=Modulator_1, children [Random_0], properties {pitchMax=1.5, pitchMin=1.5}}
 {id=Modulator_2, children [Random_1], properties {pitchMax=1.5, pitchMin=1.5, volumeMax=0.5, volumeMin=0.5}}
 {id=Modulator_3, children [Random_2], properties {pitchMax=1, pitchMin=1, volumeMax=0.9, volumeMin=0.8}}
 {id=Random_0, children [WavePlayer_1,WavePlayer_2,WavePlayer_3], properties {weights [1,1,1]}}
 {id=Random_1, children [WavePlayer_5,WavePlayer_6,WavePlayer_7,WavePlayer_8,WavePlayer_9], properties {weights [1,1,1,1,1]}}
 {id=Random_2, children [WavePlayer_10, WavePlayer_11, WavePlayer_12], properties {weights [1,1,1]}}
 {id=WavePlayer_1, properties {soundWaveAssetPtr=&/Reality_Seed_Land_Whoosh_03}}
 {id=WavePlayer_10, properties {soundWaveAssetPtr=&/Seed_Throw_01}}
 {id=WavePlayer_11, properties {soundWaveAssetPtr=&/Seed_Throw_02}}
 {id=WavePlayer_12, properties {soundWaveAssetPtr=&/Seed_Throw_03}}
 {id=WavePlayer_2, properties {soundWaveAssetPtr=&/Reality_Seed_Land_Whoosh_01}}
 {id=WavePlayer_3, properties {soundWaveAssetPtr=&/Reality_Seed_Land_Whoosh_02}}
 {id=WavePlayer_5, properties {soundWaveAssetPtr=&/Reality_Seed_Launch_01}}
 {id=WavePlayer_6, properties {soundWaveAssetPtr=&/Reality_Seed_Launch_02}}
 {id=WavePlayer_7, properties {soundWaveAssetPtr=&/Reality_Seed_Launch_03}}
 {id=WavePlayer_8, properties {soundWaveAssetPtr=&/Reality_Seed_Launch_04}}
 {id=WavePlayer_9, properties {soundWaveAssetPtr=&/Reality_Seed_Launch_05}}
]
properties {
 attenuationSettings=&Foley_Attenuation_Basic_Att
 bOverrideAttenuation=true
 maxDistance=200
 soundClassObject=&Fort_Player_Foley_Athena
 volumeMultiplier=3
 attenuationOverrides {falloffDistance=100, stereoSpread=50, attenuationShapeExtents {x=100}}
}
