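# Sound cue for pulling a Reality Seed weed, built from Reality Seed pickup sounds.
# A Random node picks one of five pickup wave variations with equal weight, and a Modulator at the root varies the result.
# It uses the reality pods prop attenuation and plays in the gadget items sound class.
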SoundCue /RealitySeedGameplay/items/RealitySeed/Sounds/RealitySeed_Weed_Pull_Cue
root=Modulator_1
nodes [
 {id=Modulator_1, children [Random_0]}
 {id=Random_0, children [WavePlayer_0,WavePlayer_1,WavePlayer_3,WavePlayer_4,WavePlayer_5], properties {weights [1,1,1,1,1]}}
 {id=WavePlayer_0, properties {soundWaveAssetPtr=&/RealitySeed_Pickup_01}}
 {id=WavePlayer_1, properties {soundWaveAssetPtr=&/RealitySeed_Pickup_02}}
 {id=WavePlayer_3, properties {soundWaveAssetPtr=&/RealitySeed_Pickup_03}}
 {id=WavePlayer_4, properties {soundWaveAssetPtr=&/RealitySeed_Pickup_04}}
 {id=WavePlayer_5, properties {soundWaveAssetPtr=&/RealitySeed_Pickup_05}}
]
properties {
 attenuationSettings=/Game/Sounds/Attenuation/Fort_Prop_RealityPods_Att
 maxDistance=1600
 soundClassObject=&Fort_Gadgets_Items
 volumeMultiplier=1
}
